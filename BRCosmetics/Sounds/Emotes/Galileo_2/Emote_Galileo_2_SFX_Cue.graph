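# Sound cue for the Galileo_2 emote's SFX, using Fortnite's Athena emote attenuation and the emote music sound class.
# It plays a single wave after a short fixed delay, with no random variation.
# It references the shared emote foley concurrency rule to limit overlapping instances and has a send to the music audio-analysis submix.
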SoundCue /BRCosmetics/Sounds/Emotes/Galileo_2/Emote_Galileo_2_SFX_Cue
root=Delay_0
nodes [
 {id=Delay_0, children [WavePlayer_0], properties {delayMax=0.2, delayMin=0.2}}
 {id=WavePlayer_0, properties {soundWaveAssetPtr=&/Emote_Galileo_2_SFX}}
]
properties {
 attenuationSettings=&Fort_Emotes_Athena_Att
 maxDistance=1600
 soundClassObject=&Fort_Music_Emote
 volumeMultiplier=1
 concurrencyOverrides {maxCount=3, resolutionRule=StopOldest}
 concurrencySet [/Game/Sounds/Concurrency_Rules/Athena_EmoteFoley_Conc]
 soundSubmixSends [{minSendLevel=0, soundSubmix=&Fortnite_AudioAnalysis_Music}]
}
